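# Animation Blueprint for the side-action revolver weapon that steps its drum through six chamber positions as it fires.
# A fire sequence passes through DefaultSlot and gets the RotateChamber state machine's output applied additively.
# RotateChamber starts at an additive identity pose, then cycles through six one-shot additive rotation states.
# The Revolver_RotChamber notify increments a counter that sets each transition flag, resetting counter and earlier flags at position six.
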AnimBlueprint /Game/Animation/Game/MainPlayer/Combat/Guns/FuturisticRevolver/Pistol_Revolver_SideAction_AnimBP1
target_skeleton=/Game/Weapons/Skeleton_Templates/SK_Revolver_SideAction_Skeleton
graphs [
 AnimGraph {nodes [
   n1 = Sequence Player 'Pistol_Revlover_SideAction_Fire_WMM'() {sequence=/Game/Animation/Game/MainPlayer/Combat/Guns/Pistol/Pistol_Revlover_SideAction_Fire_WMM}
   AnimGraphNode_StateMachine_0 = RotateChamber() {is=title AnimGraphNode_StateMachine}
   Output Pose(Result=Apply Additive(Base=Slot 'DefaultSlot'(Source=@n1), Additive=@AnimGraphNode_StateMachine_0) {alphaBoolBlend {blendOption=Linear}}) {layerGroup=None}
  ]}
 RotateChamber {
  entry=Default
  path=AnimGraph.AnimGraphNode_StateMachine_0.RotateChamber
  ids=id
  nodes [
   AnimStateNode_0 = Default() {is=title AnimStateNode}
   AnimStateNode_1 = DrumPosition6() {is=title AnimStateNode}
   AnimStateNode_2 = DrumPosition1() {is=title AnimStateNode}
   AnimStateNode_3 = DrumPosition2() {is=title AnimStateNode}
   AnimStateNode_4 = DrumPosition3() {is=title AnimStateNode}
   AnimStateNode_5 = DrumPosition4() {is=title AnimStateNode}
   AnimStateNode_6 = DrumPosition5() {is=title AnimStateNode}
   AnimStateTransitionNode_1 = AnimStateTransitionNode() {crossfadeDuration=0.033}
   AnimStateTransitionNode_2 = AnimStateTransitionNode() {crossfadeDuration=0.033}
   AnimStateTransitionNode_3 = AnimStateTransitionNode() {crossfadeDuration=0.033}
   AnimStateTransitionNode_4 = AnimStateTransitionNode() {crossfadeDuration=0.033}
   AnimStateTransitionNode_5 = AnimStateTransitionNode() {crossfadeDuration=0.033}
   AnimStateTransitionNode_6 = AnimStateTransitionNode() {crossfadeDuration=0.033}
   AnimStateTransitionNode_7 = AnimStateTransitionNode()
  ]
  links [
   {
    dst=DrumPosition1
    rule=AnimGraph.AnimGraphNode_StateMachine_0.RotateChamber.AnimStateTransitionNode_7.Transition
    src=Default
    via=AnimStateTransitionNode_7
   }
   {
    dst=DrumPosition2
    rule=AnimGraph.AnimGraphNode_StateMachine_0.RotateChamber.AnimStateTransitionNode_2.AnimationTransitionGraph_3
    src=DrumPosition1
    via=AnimStateTransitionNode_2
   }
   {
    dst=DrumPosition3
    rule=AnimGraph.AnimGraphNode_StateMachine_0.RotateChamber.AnimStateTransitionNode_3.AnimationTransitionGraph_3
    src=DrumPosition2
    via=AnimStateTransitionNode_3
   }
   {
    dst=DrumPosition4
    rule=AnimGraph.AnimGraphNode_StateMachine_0.RotateChamber.AnimStateTransitionNode_4.AnimationTransitionGraph_3
    src=DrumPosition3
    via=AnimStateTransitionNode_4
   }
   {
    dst=DrumPosition5
    rule=AnimGraph.AnimGraphNode_StateMachine_0.RotateChamber.AnimStateTransitionNode_5.AnimationTransitionGraph_3
    src=DrumPosition4
    via=AnimStateTransitionNode_5
   }
   {
    dst=DrumPosition6
    rule=AnimGraph.AnimGraphNode_StateMachine_0.RotateChamber.AnimStateTransitionNode_6.AnimationTransitionGraph_3
    src=DrumPosition5
    via=AnimStateTransitionNode_6
   }
   {
    dst=DrumPosition1
    rule=AnimGraph.AnimGraphNode_StateMachine_0.RotateChamber.AnimStateTransitionNode_1.AnimationTransitionGraph_3
    src=DrumPosition6
    via=AnimStateTransitionNode_1
   }
  ]
 }
 Default {
  path=AnimGraph.AnimGraphNode_StateMachine_0.RotateChamber.AnimStateNode_0.Default
  nodes [Output Animation Pose(Result=Additive Identity Pose())]
 }
 DrumPosition6 {
  path=AnimGraph.AnimGraphNode_StateMachine_0.RotateChamber.AnimStateNode_1.DrumPosition6
  nodes [
   n1 = "Sequence Player (additive) 'FuturisticRevolver_Fire_ChamberRot_6_CMM_Wep'"() {bLoopAnimation=false, sequence=&/FuturisticRevolver_Fire_ChamberRot_6_CMM_Wep}
   Output Animation Pose(Result=@n1)
  ]
 }
 DrumPosition1 {
  path=AnimGraph.AnimGraphNode_StateMachine_0.RotateChamber.AnimStateNode_2.DrumPosition1
  nodes [
   n1 = "Sequence Player (additive) 'FuturisticRevolver_Fire_ChamberRot_1_CMM_Wep'"() {bLoopAnimation=false, sequence=&/FuturisticRevolver_Fire_ChamberRot_1_CMM_Wep}
   Output Animation Pose(Result=@n1)
  ]
 }
 DrumPosition2 {
  path=AnimGraph.AnimGraphNode_StateMachine_0.RotateChamber.AnimStateNode_3.DrumPosition2
  nodes [
   n1 = "Sequence Player (additive) 'FuturisticRevolver_Fire_ChamberRot_2_CMM_Wep'"() {bLoopAnimation=false, sequence=&/FuturisticRevolver_Fire_ChamberRot_2_CMM_Wep}
   Output Animation Pose(Result=@n1)
  ]
 }
 DrumPosition3 {
  path=AnimGraph.AnimGraphNode_StateMachine_0.RotateChamber.AnimStateNode_4.DrumPosition3
  nodes [
   n1 = "Sequence Player (additive) 'FuturisticRevolver_Fire_ChamberRot_3_CMM_Wep'"() {bLoopAnimation=false, sequence=&/FuturisticRevolver_Fire_ChamberRot_3_CMM_Wep}
   Output Animation Pose(Result=@n1)
  ]
 }
 DrumPosition4 {
  path=AnimGraph.AnimGraphNode_StateMachine_0.RotateChamber.AnimStateNode_5.DrumPosition4
  nodes [
   n1 = "Sequence Player (additive) 'FuturisticRevolver_Fire_ChamberRot_4_CMM_Wep'"() {bLoopAnimation=false, sequence=&/FuturisticRevolver_Fire_ChamberRot_4_CMM_Wep}
   Output Animation Pose(Result=@n1)
  ]
 }
 DrumPosition5 {
  path=AnimGraph.AnimGraphNode_StateMachine_0.RotateChamber.AnimStateNode_6.DrumPosition5
  nodes [
   n1 = "Sequence Player (additive) 'FuturisticRevolver_Fire_ChamberRot_5_CMM_Wep'"() {bLoopAnimation=false, sequence=&/FuturisticRevolver_Fire_ChamberRot_5_CMM_Wep}
   Output Animation Pose(Result=@n1)
  ]
 }
 AnimationTransitionGraph_3 {
  path=AnimGraph.AnimGraphNode_StateMachine_0.RotateChamber.AnimStateTransitionNode_1.AnimationTransitionGraph_3
  nodes [Result(bCanEnterTransition=Get ShouldRotateToPos1() -> (ShouldRotateToPos1:bool))]
 }
 AnimationTransitionGraph_3 {
  path=AnimGraph.AnimGraphNode_StateMachine_0.RotateChamber.AnimStateTransitionNode_2.AnimationTransitionGraph_3
  nodes [Result(bCanEnterTransition=Get ShouldRotateToPos2() -> (ShouldRotateToPos2:bool))]
 }
 AnimationTransitionGraph_3 {
  path=AnimGraph.AnimGraphNode_StateMachine_0.RotateChamber.AnimStateTransitionNode_3.AnimationTransitionGraph_3
  nodes [Result(bCanEnterTransition=Get ShouldRotateToPos3() -> (ShouldRotateToPos3:bool))]
 }
 AnimationTransitionGraph_3 {
  path=AnimGraph.AnimGraphNode_StateMachine_0.RotateChamber.AnimStateTransitionNode_4.AnimationTransitionGraph_3
  nodes [Result(bCanEnterTransition=Get ShouldRotateToPos4() -> (ShouldRotateToPos4:bool))]
 }
 AnimationTransitionGraph_3 {
  path=AnimGraph.AnimGraphNode_StateMachine_0.RotateChamber.AnimStateTransitionNode_5.AnimationTransitionGraph_3
  nodes [Result(bCanEnterTransition=Get ShouldRotateToPos5() -> (ShouldRotateToPos5:bool))]
 }
 AnimationTransitionGraph_3 {
  path=AnimGraph.AnimGraphNode_StateMachine_0.RotateChamber.AnimStateTransitionNode_6.AnimationTransitionGraph_3
  nodes [Result(bCanEnterTransition=Get ShouldRotateToPos6() -> (ShouldRotateToPos6:bool))]
 }
 Transition {
  path=AnimGraph.AnimGraphNode_StateMachine_0.RotateChamber.AnimStateTransitionNode_7.Transition
  nodes [Result(bCanEnterTransition=Get ShouldRotateToPos1())]
 }
 EventGraph {nodes [
   AnimNotify_Revolver_RotChamber() {is=title K2Node_Event}
   > n1 = Increment Int(Value=Get DrumPositionCount() -> (DrumPositionCount:int=0))
   > n2 = Switch on Int!(execute:Exec, Selection:int=@n1) -> (Default:Exec, 0:Exec, 1:Exec, 2:Exec, 3:Exec, 4:Exec, 5:Exec, 6:Exec)
   Set ShouldRotateToPos1(execute=[@n2.1, @n2.Default], ShouldRotateToPos1:bool=true) -> (Output_Get:bool)
   > Set ShouldRotateToPos6(ShouldRotateToPos6:bool=false) -> (Output_Get:bool)
   Set ShouldRotateToPos2(execute=@n2.2, ShouldRotateToPos2:bool=true) -> (Output_Get:bool)
   Set ShouldRotateToPos3(execute=@n2.3, ShouldRotateToPos3:bool=true) -> (Output_Get:bool)
   Set ShouldRotateToPos4(execute=@n2.4, ShouldRotateToPos4:bool=true) -> (Output_Get:bool)
   Set ShouldRotateToPos5(execute=@n2.5, ShouldRotateToPos5:bool=true) -> (Output_Get:bool)
   Set ShouldRotateToPos6(execute=@n2.6, ShouldRotateToPos6=true)
   > Set DrumPositionCount(DrumPositionCount:int=0) -> (Output_Get:int=0)
   > Set ShouldRotateToPos1(ShouldRotateToPos1=false)
   > Set ShouldRotateToPos2(ShouldRotateToPos2=false)
   > Set ShouldRotateToPos3(ShouldRotateToPos3=false)
   > Set ShouldRotateToPos4(ShouldRotateToPos4=false)
   > Set ShouldRotateToPos5(ShouldRotateToPos5=false)
  ]}
]
variables [
 DrumPositionCount {type=integer}
 ShouldRotateToPos1 {type=boolean}
 ShouldRotateToPos2 {type=boolean}
 ShouldRotateToPos3 {type=boolean}
 ShouldRotateToPos4 {type=boolean}
 ShouldRotateToPos5 {type=boolean}
 ShouldRotateToPos6 {type=boolean}
]
